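# Translucent UI material that draws the BondShock circle on the minimap and fullscreen map, with a separate "final medallion" look when its radius is near zero.
# Center, radius and map-scale parameters drive a distance-based SDF that yields a sharp fill and outline sized per map mode.
# A time-based sine pulses a radial-gradient glow, with a two-color gradient outline and tinted fill.
# MPC_Minimap's ScaleFactor rescales custom UVs.
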Material /WildEstate/Gameplay/BondShock/Assets/UI/Materials/M_MiniMapCircle_BondShock
graph {
 properties {blendMode=BLEND_Translucent, materialDomain=MD_UI, numCustomizedUVs=1, pixelDepthOffsetMode=PDOM_Legacy, usedShadingModels=Default Lit}
 nodes [
  n1 = VectorParameter() {group=Color, parameterName=Outline Color 2, defaultValue=<rgba 0.9805 1 0.1936 1>}
  n2 = MaterialFunctionCall(GradientOffset=Multiply(A=NamedRerouteUsage()) {constB=1.5}) {materialFunction=&MF_UI_LinearGradient}
  n3 = VectorParameter() {group=Color, parameterName=Outline Color 1, defaultValue=<rgba 1 0.1608 0 1>}
  Add(
   A=Multiply(A=NamedRerouteUsage(), B=LinearInterpolate(A=@n3.RGB, B=@n1.RGB, Alpha=@n2.Result))
   B=Multiply(A=NamedRerouteUsage(), B=VectorParameter() {group=Color, parameterName=Circle Fill Color, defaultValue=<rgba 1 0.8 0.2 1>}.RGB)
  )
  n4 = NamedRerouteUsage()
  n5 = NamedRerouteUsage()
  n6 = LinearInterpolate(Alpha=NamedRerouteUsage()) {constA=1.5, constAlpha=1, constB=6}
  n7 = LinearInterpolate(Alpha=NamedRerouteUsage()) {constA=0.75, constAlpha=1, constB=1.25}
  n8 = Add(A=Sine(None=Time()) {period=4})
  n9 = Multiply(
   A=Multiply(A=NamedRerouteUsage())
   B=Multiply(
    A=MaterialFunctionCall(UVs=@n4, CenterPosition=@n5, Radius=@n6, Density=@n7) {materialFunction=&RadialGradientExponential}.RadialGradientExponential
    B=Multiply(A=@n8) {constB=0.5}
   )
  )
  n10 = NamedRerouteUsage()
  n11 = LinearInterpolate(A=Add(A=Multiply(A=@n8) {constB=0.175}) {constB=0.25}, Alpha=NamedRerouteUsage()) {constA=0.75, constAlpha=1}
  Multiply(A=Add(A=@n9, B=Multiply(A=@n10, B=@n11)))
  n12 = Subtract(A=TextureCoordinate(), B=NamedRerouteUsage())
  n13 = NamedRerouteUsage()
  NamedRerouteDeclaration(None=Add(A=Divide(A=@n12, B=@n13)) {constB=0.5}) {name=UV, nodeColor=<rgba 0 0 0 1>}
  n14 = NamedRerouteUsage()
  NamedRerouteDeclaration(None=Saturate(None=Subtract(A=@n9, B=@n14))) {name=CircleFill, nodeColor=<rgba 0 0 0 1>}
  n15 = VectorParameter() {
   desc="Circle Origin.X, Origin.Y, Radius (all multiplied by map scale).\r\nA is ignored"
   group=Size
   parameterName=CircleCenterAndRadius
   defaultValue=<rgba 0.5 0.5 1.6763 1>
  }
  n16 = ComponentMask(None=@n15.RGB) {g=true, r=true}
  n17 = NamedRerouteUsage()
  n18 = NamedRerouteUsage()
  n19 = Subtract(
   A=Distance(A=TextureCoordinate(), B=NamedRerouteDeclaration(None=@n16) {name=CirclePosition, nodeColor=<rgba 0 0 0 1>})
   B=NamedRerouteDeclaration(None=LinearInterpolate(A=@n15.B, B=@n17, Alpha=@n18)) {name=CircleSize, nodeColor=<rgba 0 0 0 1>}
  )
  n20 = SmoothStep(Value=@n19) {constMax=0.001}
  NamedRerouteDeclaration(None=OneMinus(None=@n20)) {name=Fill_Sharp, nodeColor=<rgba 0 0 0 1>}
  n21 = VectorParameter() {desc="MapSize, Scale, IsFullscreenMap, A is ignored", group=Size, parameterName=MapSizeAndScale, defaultValue=<rgba 0.5 0.5 1.6763 1>}
  n22 = NamedRerouteDeclaration(None=@n21.B) {name=IsFullscreenMap, desc=Is this the fullscreen map or the minimap, nodeColor=<rgba 0 0 0 1>}
  NamedRerouteDeclaration(None=Multiply(A=NamedRerouteDeclaration(None=@n21.G) {name=MapScale, nodeColor=<rgba 0 0 0 1>}, B=LinearInterpolate(Alpha=@n22) {constA=6, constB=10}) {constB=10}) {name=FinalMedallionMapScale, nodeColor=<rgba 0 0 0 1>}
  n23 = NamedRerouteUsage()
  n24 = Constant() {r=0.01}
  n25 = LinearInterpolate(
   A=Multiply(A=NamedRerouteUsage(), B=LinearInterpolate(Alpha=@n23) {constA=1.2, constB=1.5}) {constB=1.5}
   B=Multiply(A=NamedRerouteUsage()) {constB=0.2}
   Alpha=NamedRerouteDeclaration(None=MaterialFunctionCall!(A=@n15.B, B=@n24) {materialFunction=/FortUILibrary/MaterialFunctions/Basics/MF_UI_BooleanOperations}.IsLessOrEqual) {name=IsFinalMedallion, nodeColor=<rgba 0 0 0 1>}
  ) {constA=0.0015, constB=0}
  n26 = Abs(None=@n19)
  NamedRerouteDeclaration(None=OneMinus(None=SmoothStep(Max=@n25, Value=@n26) {constMax=0.0015, constValue=1})) {name=Outline_Sharp, nodeColor=<rgba 0 0 0 1>}
  NamedRerouteDeclaration(None=@n20) {name=SDF, nodeColor=<rgba 0 0 0 1>}
  NamedRerouteDeclaration(None=Add(A=NamedRerouteUsage(), B=NamedRerouteUsage())) {name=Fill, nodeColor=<rgba 0 0 0 1>}
  n27 = TextureCoordinate()
  n28 = CollectionParameter() {collection=/Game/Athena/HUD/MiniMap/MPC_Minimap, parameterName=ScaleFactor}
  n29 = OneMinus(None=@n28)
  NamedRerouteDeclaration(None=Add(A=Multiply(A=@n27, B=@n28) {constB=2}, B=Multiply(A=@n29) {constB=0.5})) {name=CustomUV, nodeColor=<rgba 0 0 0 1>}
  NamedRerouteUsage()
 ]
}
