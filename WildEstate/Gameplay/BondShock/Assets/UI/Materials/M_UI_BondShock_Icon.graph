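# M_UI_BondShock_Icon is a translucent User Interface material for the BondShock popup icon in WildEstate.
# Its graph holds a single TextureSample node referencing T_UI_BondShock_PopupIcon, with no parameters or other math.
# No connections from that sample to the material outputs appear in the graph.
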Material /WildEstate/Gameplay/BondShock/Assets/UI/Materials/M_UI_BondShock_Icon
graph {
 properties {bAutomaticallySetUsageInEditor=false, blendMode=BLEND_Translucent, materialDomain=MD_UI, usedShadingModels=Default Lit}
 nodes [TextureSample() {texture=/WildEstate/Gameplay/BondShock/Assets/UI/Texture/T_UI_BondShock_PopupIcon}]
}
