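# BondShock alert widget that shows a headshot-damage notice about the Jebediah head flying off and becoming available to pick up again.
# It wraps a single notification-with-backplate visuals widget, setting its headline, details and icon material from variables in PreConstruct.
# On initialization it waits a few seconds, then clears the current-gameplay-alert-changed event state on the global event router, presumably ending the alert.
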WidgetBlueprint /WildEstate/Gameplay/BondShock/Assets/UI/WBP_BondShock_Alert
graphs [EventGraph {nodes [
   Event PreConstruct()
   n1 = Get WBP_NotificationWithBackplate_Visuals() -> (WBP_NotificationWithBackplate_Visuals:WBP Notification With Backplate Visuals Object Reference)
   > SetHeadling!(execute, self:WBP Notification With Backplate Visuals Object Reference=@n1, InText:Text=Get HeadlingText() -> (HeadlingText:Text)) -> (then)
   > SetDetails!(execute, self:WBP Notification With Backplate Visuals Object Reference=@n1, InText:Text=Get DetailsText() -> (DetailsText:Text)) -> (then)
   > SetBrushFromMaterial(
    self:WBP Notification With Backplate Visuals Object Reference=@n1
    Material=Get IconMaterial() -> (IconMaterial:Material Interface Object Reference)
   )
   Event OnInitialized()
   > Delay(Duration=5)
   > ClearEventState!(
    execute
    self:Gameplay Event Router Component Object Reference=GetGlobalEventRouter()
    EventType:Script Struct Object Reference=/Script/BattleRoyaleGameplayAlerts.FortEvent_CurrentGameplayAlertChanged
   ) -> (then)
  ]}]
variables [
 DetailsText {type=string, default="The first headshot damage will hit the Jebediah head. Once it flies off, it will be available to pick up again"}
 GameplayAlertChangedEvent {type=FortEvent_CurrentGameplayAlertChanged, default {gameplayAlertTag {tagName=WildEstate.BondShock.Alert}}}
 HeadlingText {type=string, default=HEADSHOT DAMAGE}
 IconMaterial {type=MaterialInterface, default=/WildEstate/Gameplay/BondShock/Assets/UI/Materials/M_UI_BondShock_Icon}
]
widgets {
 info {
  inheritedWidgetCount=0
  parentClass=&UserWidget
  rootWidgetClass=/Game/Athena/HUD/Messaging/WBP_NotificationWithBackplate_Visuals.WBP_NotificationWithBackplate_Visuals_C
  widgetCount=1
 }
 widgets [WBP_NotificationWithBackplate_Visuals {
   widgetClassPath=/Game/Athena/HUD/Messaging/WBP_NotificationWithBackplate_Visuals
   bIsVariable=true
   widget=WBP_NotificationWithBackplate_Visuals
  }]
}
